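# Unlit translucent particle material for floating music notes in the front-end phonograph effect, usable on Cascade and Niagara sprites.
# Particle color feeds an emissive adjustment function, and particle alpha masks the red channel of a SubUV music-note texture.
# A sine of time, shifted by the TimeOffset dynamic parameter and scaled by BobbleStrength, builds a Z-axis bobbing vector that each emitter can drive.
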Material /Game/Environments/FrontEnd/FX/Phonograph/M_ParticleNotes
graph {
 properties {
  bUsedWithNiagaraSprites=true
  bUsedWithParticleSprites=true
  blendMode=BLEND_Translucent
  pixelDepthOffsetMode=PDOM_Legacy
  shadingModel=MSM_Unlit
  usedShadingModels=Unlit
 }
 nodes [
  n1 = ParticleColor()
  MaterialFunctionCall(Emissive=@n1.RGB) {materialFunction=&MF_FX_EmissiveAdjustments}
  Multiply(
   A=@n1.A
   B=TextureSampleParameterSubUV!(UVs, Apply View MipBias) {parameterName=ParticleTexture, texture=/Game/Weapons/FORT_Melee/Effects/Spear_BoomBox/T_FX_MusicNotes_01}.R
  )
  n2 = DynamicParameter() {defaultValue=<rgba 1 16 1 1>, paramNames [TimeOffset, BobbleStrength, Param3, Param4]}
  n3 = Add(A=@n2.TimeOffset, B=Time())
  Multiply(A=@n2.BobbleStrength, B=Multiply(A=Sine(None=@n3), B=Constant3Vector() {constant=<rgba 0 0 1 0>}))
 ]
}
